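# UI-domain translucent material for the first-place placement plate on the score HUD of a limited-time mode.
# It blends the RGB of the plate texture parameter toward solid white, using the FillAmount scalar parameter as the lerp alpha.
# Raising FillAmount from zero to one therefore washes the plate art out to white.
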Material /Game/Athena/HUD/Materials/Score/M_ScoreLTM_PlacementPlate_1st
graph {
 properties {blendMode=BLEND_Translucent, materialDomain=MD_UI, pixelDepthOffsetMode=PDOM_Legacy, usedShadingModels=Default Lit}
 nodes [LinearInterpolate(
   A=TextureSampleParameter2D() {parameterName=Texture, texture=&/T_ScoreLTM_PlacementPlate_1st}.RGB
   B=Constant3Vector() {constant=<rgba 1 1 1 0>}
   Alpha=ScalarParameter() {parameterName=FillAmount}
  )]
}
